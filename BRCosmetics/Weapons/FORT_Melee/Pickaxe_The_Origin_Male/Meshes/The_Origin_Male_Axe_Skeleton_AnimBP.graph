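# Anim Blueprint for The Origin (Male) pickaxe that separates and floats its base, blade and bottom pieces while equipped.
# Each update, a melee weapon animation on the player or front-end state sets equipAlpha to 1 after a delay, otherwise zero.
# At full alpha, time-driven random vectors update the float offsets.
# Chained additive Modify Bone nodes apply fixed spread, then float, with interpolated alpha; bottom float variables drive opposite-side bones.
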AnimBlueprint /BRCosmetics/Weapons/FORT_Melee/Pickaxe_The_Origin_Male/Meshes/The_Origin_Male_Axe_Skeleton_AnimBP
target_skeleton=&/The_Origin_Male_Axe_Skeleton
graphs [
 AnimGraph {nodes [
   n1 = Get equipAlpha() -> (equipAlpha:double)
   n2 = "Transform (Modify) Bone - Bone: base_l"(ComponentPose=Local To Component(LocalPose=Input Pose()), Alpha=@n1, Translation="-0.000,-3,1.5", Rotation="0,0,0", Scale="1,1,1") {
    rotationMode=BMM_Additive
    translationMode=BMM_Additive
    translationSpace=BCS_BoneSpace
    alphaBoolBlend {blendOption=Linear}
    alphaScaleBiasClamp {bInterpResult=true, interpSpeedDecreasing=20, scale=0.75}
    boneToModify {boneName=base_l}
    translation {y=-3, z=1.5}
   }
   n3 = "Transform (Modify) Bone - Bone: blade_l"(ComponentPose=@n2, Alpha=@n1, Translation="0,0,3", Rotation="0,0,0", Scale="1,1,1") {
    rotationMode=BMM_Additive
    translationMode=BMM_Additive
    translationSpace=BCS_BoneSpace
    alphaBoolBlend {blendOption=Linear}
    alphaScaleBiasClamp {bInterpResult=true, interpSpeedDecreasing=20, scale=0.75}
    boneToModify {boneName=blade_l}
    translation {z=3}
   }
   n4 = "Transform (Modify) Bone - Bone: base_r"(ComponentPose=@n3, Alpha=@n1, Translation="0,3,1.5", Rotation="0,0,0", Scale="1,1,1") {
    rotationMode=BMM_Additive
    translationMode=BMM_Additive
    translationSpace=BCS_BoneSpace
    alphaBoolBlend {blendOption=Linear}
    alphaScaleBiasClamp {bInterpResult=true, interpSpeedDecreasing=20, scale=0.75}
    boneToModify {boneName=base_r}
    translation {y=3, z=1.5}
   }
   n5 = "Transform (Modify) Bone - Bone: blade_r"(ComponentPose=@n4, Alpha=@n1, Translation="0,0,3", Rotation="0,0,0", Scale="1,1,1") {
    rotationMode=BMM_Additive
    translationMode=BMM_Additive
    translationSpace=BCS_BoneSpace
    alphaBoolBlend {blendOption=Linear}
    alphaScaleBiasClamp {bInterpResult=true, interpSpeedDecreasing=20, interpSpeedIncreasing=3, scale=0.75}
    boneToModify {boneName=blade_r}
    translation {z=3}
   }
   n6 = "Transform (Modify) Bone - Bone: bottom_r"(ComponentPose=@n5, Alpha=@n1, Translation="0,3,-1.5", Rotation="0,0,0", Scale="1,1,1") {
    rotationMode=BMM_Additive
    translationMode=BMM_Additive
    translationSpace=BCS_BoneSpace
    alphaBoolBlend {blendOption=Linear}
    alphaScaleBiasClamp {bInterpResult=true, interpSpeedDecreasing=20, scale=0.75}
    boneToModify {boneName=bottom_r}
    translation {y=3, z=-1.5}
   }
   n7 = "Transform (Modify) Bone - Bone: bottom_l"(ComponentPose=@n6, Alpha=@n1, Translation="0,-3,-1.5", Rotation="0,0,0", Scale="1,1,1") {
    rotationMode=BMM_Additive
    translationMode=BMM_Additive
    translationSpace=BCS_BoneSpace
    alphaBoolBlend {blendOption=Linear}
    alphaScaleBiasClamp {bInterpResult=true, interpSpeedDecreasing=20, interpSpeedIncreasing=3, scale=0.75}
    boneToModify {boneName=bottom_l}
    translation {y=-3, z=-1.5}
   }
   n8 = Get equipAlpha()
   n9 = "Transform (Modify) Bone - Bone: base_l"(ComponentPose=@n7, Alpha=@n8, Translation=Get base_float_l() -> (base_float_l:Vector), Rotation="0,0,0", Scale="1,1,1") {
    rotationMode=BMM_Additive
    translationMode=BMM_Additive
    translationSpace=BCS_BoneSpace
    alphaBoolBlend {blendOption=Linear}
    alphaScaleBiasClamp {bInterpResult=true, interpSpeedDecreasing=20, scale=1.5}
    boneToModify {boneName=base_l}
    translation {y=-3, z=-1.5}
   }
   n10 = "Transform (Modify) Bone - Bone: base_r"(ComponentPose=@n9, Alpha=@n8, Translation=Get base_float_r() -> (base_float_r:Vector), Rotation="0,0,0", Scale="1,1,1") {
    rotationMode=BMM_Additive
    translationMode=BMM_Additive
    translationSpace=BCS_BoneSpace
    alphaBoolBlend {blendOption=Linear}
    alphaScaleBiasClamp {bInterpResult=true, interpSpeedDecreasing=20, scale=1.5}
    boneToModify {boneName=base_r}
    translation {y=-3, z=-1.5}
   }
   n11 = "Transform (Modify) Bone - Bone: blade_l"(ComponentPose=@n10, Alpha=@n8, Translation=Get blade_float_l() -> (blade_float_l:Vector), Rotation="0,0,0", Scale="1,1,1") {
    rotationMode=BMM_Additive
    translationMode=BMM_Additive
    translationSpace=BCS_BoneSpace
    alphaBoolBlend {blendOption=Linear}
    alphaScaleBiasClamp {bInterpResult=true, interpSpeedDecreasing=20, scale=1.5}
    boneToModify {boneName=blade_l}
    translation {y=-3, z=-1.5}
   }
   n12 = "Transform (Modify) Bone - Bone: blade_r"(ComponentPose=@n11, Alpha=@n8, Translation=Get blade_float_r() -> (blade_float_r:Vector), Rotation="0,0,0", Scale="1,1,1") {
    rotationMode=BMM_Additive
    translationMode=BMM_Additive
    translationSpace=BCS_BoneSpace
    alphaBoolBlend {blendOption=Linear}
    alphaScaleBiasClamp {bInterpResult=true, interpSpeedDecreasing=20, scale=1.5}
    boneToModify {boneName=blade_r}
    translation {y=-3, z=-1.5}
   }
   n13 = "Transform (Modify) Bone - Bone: bottom_r"(ComponentPose=@n12, Alpha=@n8, Translation=Get bottom_float_l() -> (bottom_float_l:Vector), Rotation="0,0,0", Scale="1,1,1") {
    rotationMode=BMM_Additive
    translationMode=BMM_Additive
    translationSpace=BCS_BoneSpace
    alphaBoolBlend {blendOption=Linear}
    alphaScaleBiasClamp {bInterpResult=true, interpSpeedDecreasing=20, scale=1.5}
    boneToModify {boneName=bottom_r}
    translation {y=-3, z=-1.5}
   }
   n14 = "Transform (Modify) Bone - Bone: bottom_l"(ComponentPose=@n13, Alpha=@n8, Translation=Get bottom_float_r() -> (bottom_float_r:Vector), Rotation="0,0,0", Scale="1,1,1") {
    rotationMode=BMM_Additive
    translationMode=BMM_Additive
    translationSpace=BCS_BoneSpace
    alphaBoolBlend {blendOption=Linear}
    alphaScaleBiasClamp {bInterpResult=true, interpSpeedDecreasing=20, scale=1.5}
    boneToModify {boneName=bottom_l}
    translation {y=-3, z=-1.5}
   }
   Output Pose(Result=Component To Local(ComponentPose=@n14)) {layerGroup=None}
  ]}
 EventGraph {nodes [
   Event BlueprintUpdateAnimation()
   > n1 = Sequence!(execute) -> (then_0, then_1, then_2:Exec, then_3:Exec)
   > n2 = Cast To FortPlayerAnimInstance(Object=GetAnimInstance(self=GetOwnerSkeletalMeshComponent()))
   > n3 = Switch on EFortWeaponCoreAnimation!(execute, Selection:EFortWeaponCoreAnimation Enum=Get WeaponCoreAnim(self=@n2)) -> (
    Melee:Exec
    Pistol:Exec
    Shotgun:Exec
    PaperBlueprint:Exec
    Rifle:Exec
    MeleeOneHand:Exec
    MachinePistol:Exec
    RocketLauncher:Exec
    GrenadeLauncher:Exec
    GoingCommando:Exec
    AssaultRifle:Exec
    TacticalShotgun:Exec
    SniperRifle:Exec
    TrapPlacement:Exec
    ShoulderLauncher:Exec
    AbilityDecoTool:Exec
    Crossbow:Exec
    C4:Exec
    RemoteControl:Exec
    DualWield:Exec
    AR_BullPup:Exec
    AR_ForwardGrip:Exec
    MedPackPaddles:Exec
    SMG_P90:Exec
    AR_DrumGun:Exec
    Consumable_Small:Exec
    Consumable_Large:Exec
    Balloon:Exec
    MountedTurret:Exec
    CreativeTool:Exec
    ExplosiveBow:Exec
    AshtonIndigo:Exec
    AshtonChicago:Exec
    MeleeDualWield:Exec
    Unarmed:Exec
   )
   Set isMelee(
    execute=[
     @n3.AR_BullPup
     @n3.AR_DrumGun
     @n3.AR_ForwardGrip
     @n3.AbilityDecoTool
     @n3.AshtonChicago
     @n3.AshtonIndigo
     @n3.AssaultRifle
     @n3.Balloon
     @n3.C4
     @n3.Consumable_Large
     @n3.Consumable_Small
     @n3.CreativeTool
     @n3.Crossbow
     @n3.DualWield
     @n3.ExplosiveBow
     @n3.GoingCommando
     @n3.GrenadeLauncher
     @n3.MachinePistol
     @n3.MedPackPaddles
     @n3.MountedTurret
     @n3.PaperBlueprint
     @n3.Pistol
     @n3.RemoteControl
     @n3.Rifle
     @n3.RocketLauncher
     @n3.SMG_P90
     @n3.Shotgun
     @n3.ShoulderLauncher
     @n3.SniperRifle
     @n3.TacticalShotgun
     @n3.TrapPlacement
    ]
    isMelee:bool=false
   ) -> (Output_Get:bool)
   Set isMelee(execute=[@n3.Melee, @n3.MeleeDualWield, @n3.MeleeOneHand], isMelee=true)
   n4 = Branch(execute=@n1.then_1, Condition=OR Boolean(A=Get isMelee() -> (isMelee:bool), B=Get bIsInFrontEnd()))
   > Do Once(Reset=@n5.then)
   > Delay(Duration=0.6)
   > Set equipAlpha(equipAlpha:double=1) -> (Output_Get:double)
   Set equipAlpha(execute=@n4.else, equipAlpha=0)
   > Set base_float_l(base_float_l:Vector="0, 0, 0") -> (Output_Get:Vector)
   > Set blade_float_r(blade_float_r:Vector="0, 0, 0") -> (Output_Get:Vector)
   > Set bottom_float_l(bottom_float_l:Vector="0, 0, 0") -> (Output_Get:Vector)
   > Set bottom_float_r(bottom_float_r:Vector="0, 0, 0") -> (Output_Get:Vector)
   > Set base_float_r(base_float_r:Vector="0, 0, 0") -> (Output_Get:Vector)
   > n5 = Set blade_float_l(blade_float_l:Vector="0, 0, 0") -> (Output_Get:Vector)
   Branch(execute=@n1.then_2, Condition="Equal (Float)"(A=Get equipAlpha(), B=1))
   n6 = GetGameTimeInSeconds()
   n7 = GetOwnerSkeletalMeshComponent()
   > n8 = Random Vector Function!(
    execute
    CharacterPartAnimInstance:Custom Character Part Anim Instance Object Reference
    SkelMeshComponent:Mesh Component Object Reference=@n7
    JointName:Name=None
    GameTimeInSeconds:double=@n6
    TimeOffset:double=0.2
    PerlinScale:double=0
    Frequency:Vector="0.9,1.1,0.75"
    Amplitude:Vector="0.4,1.5,1.2"
    Offset:Vector="0,0,1"
   ) -> (then, VectorOut:Vector)
   > n9 = Random Vector Function(JointName=None, GameTimeInSeconds=@n6, TimeOffset=0.2, PerlinScale=0, Frequency="1.2,1.5,1.2", Amplitude="0.4,1.5,1.2", Offset="0,0,1")
   > n10 = Random Vector Function(
    SkelMeshComponent=@n7
    JointName=None
    GameTimeInSeconds=@n6
    TimeOffset=3
    PerlinScale=0
    Frequency="1.2,1.5,1.2"
    Amplitude="0.6,0.5,0.5"
    Offset="0,0,0"
   )
   > n11 = Random Vector Function(
    SkelMeshComponent=@n7
    JointName=None
    GameTimeInSeconds=@n6
    TimeOffset=1.1
    PerlinScale=0
    Frequency="1.5,1.2,1.5"
    Amplitude="0.6,0.5,0.5"
    Offset="0,0,0"
   )
   n12 = BreakVector(InVec=@n8)
   > Set base_float_l(base_float_l=MakeVector(X=@n12, Y=0, Z=@n12.Z))
   n13 = BreakVector(InVec=@n9)
   > Set base_float_r(base_float_r=MakeVector(X=@n13, Y=0, Z=@n13.Z))
   > Set bottom_float_l(bottom_float_l=@n10)
   > Set bottom_float_r(bottom_float_r=@n11)
   n14 = GetGameTimeInSeconds()
   > n15 = Random Vector Function(JointName=None, GameTimeInSeconds=@n14, TimeOffset=-2, PerlinScale=0, Frequency="1.2,1.5,1.2", Amplitude="0.3,1.5,1.2", Offset="0,0,0")
   > n16 = Random Vector Function(JointName=None, GameTimeInSeconds=@n14, TimeOffset=-2, PerlinScale=0, Frequency="0.9,1.5,0.7", Amplitude="0.3,1.5,1.2", Offset="0,0,0")
   n17 = BreakVector(InVec=@n15)
   > Set blade_float_r(blade_float_r=MakeVector(X=@n17, Y=0, Z=@n17.Z))
   n18 = BreakVector(InVec=@n16)
   > Set blade_float_l(blade_float_l=MakeVector(X=@n18, Y=0, Z=@n18.Z))
   TryGetPawnOwner()
  ]}
]
variables [
 base_float_l {type=Vector}
 base_float_r {type=Vector}
 blade_float_l {type=Vector}
 blade_float_r {type=Vector}
 bottom_float_l {type=Vector}
 bottom_float_r {type=Vector}
 equipAlpha {type=number}
 isMelee {type=boolean}
]
